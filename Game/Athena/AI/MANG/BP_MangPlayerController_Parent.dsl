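; Parent AI bot controller Blueprint for MANG NPCs that reacts to alert level and patrol state changes.
; Alert level changes are forwarded to the BP_MangPlayerPawn_Parent pawn to drive vocalizations and "!" FX.
; Patrol start applies a patrol GameplayEffect, stores its handle, and swaps the movement rotation rate for a patrol rate after saving the original.
; Patrol stop removes that effect by handle and restores the saved rotation rate.
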(event EventOnAlertLevelChanged
  (params (OldAlertLevel :type EAlertLevel%20Enum :default "Unaware") (NewAlertLevel :type EAlertLevel%20Enum :default "Unaware"))
  (bind _asbp_mang_player_pawn_parent (Utilities|Casting|CastToBP_MangPlayerPawn_Parent (Variables|FortAthenaAIBotController|GetPlayerBotPawn)))
  (Class|BPMangPlayerPawnParent|AlertLevelChanged _asbp_mang_player_pawn_parent OldAlertLevel NewAlertLevel) ; Drives Vocalizations, and "!" FX.
)

(event OnStartPatrollingEvent(FortAthenaNpcPatrolling)
  (bind _playerbotpawn (Variables|FortAthenaAIBotController|GetPlayerBotPawn))
  (bind _returnvalue (Ability|GetAbilitySystemComponent _playerbotpawn))
  (bind _charactermovement (Variables|/Script/Engine.Character|GetCharacterMovement _playerbotpawn))
  (Utilities|IsValid _returnvalue
    (:Is%20Valid
      (bind _returnvalue_2 (GameplayEffects|ApplyGameplayEffectToTarget :self _returnvalue :GameplayEffectClass (Variables|Default|GetGE_PatrolTag) :Target _returnvalue :Level 0.0))
      (Variables|Default|SetGE_PatrolSpeedHandle _returnvalue_2)
      (Variables|Default|SetSavedRotationRate (Variables|/Script/Engine.CharacterMovementComponent|GetRotationRate _charactermovement))
      (Variables|/Script/Engine.CharacterMovementComponent|SetRotationRate (Variables|Patrol|GetPatrollingRotationRate) _charactermovement))))

(event OnStopPatrollingEvent(FortAthenaNpcPatrolling)
  (bind _playerbotpawn (Variables|FortAthenaAIBotController|GetPlayerBotPawn))
  (bind _returnvalue (Ability|GetAbilitySystemComponent _playerbotpawn))
  (Utilities|IsValid _returnvalue
    (:Is%20Valid
      (GameplayEffects|RemoveActiveGameplayEffect _returnvalue (Variables|Default|GetGE_PatrolSpeedHandle) -1)
      (Variables|/Script/Engine.CharacterMovementComponent|SetRotationRate (Variables|Default|GetSavedRotationRate) (Variables|/Script/Engine.Character|GetCharacterMovement _playerbotpawn)))))

(fn ConstructionScript
  (|CallParentFunction))
