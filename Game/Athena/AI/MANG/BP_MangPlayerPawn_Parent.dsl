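; Parent player-pawn Blueprint for MANG henchman NPCs, handling alert presentation, downed (DBNO) state and patrol animation.
; The replicated alert level's OnRep updates the state tracker, "!"/"?" cues, escalation voice lines, animation events and stress effects that influence patrol speed.
; A linked patrol anim layer picks weapon-based sets, also supplying hit-react montages.
; It also removes bumped players' disguises, teleports out on storm damage and optionally drops ammo on death.
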(event EventEndPlay
  (params (EndPlayReason :type EEndPlayReason%20Enum :default "Destroyed"))
  (|CallParentFunction EndPlayReason)
  (Networking|CanExecuteCosmeticEvents
    (:True
      (bind _returnvalue (FortGameplayDataTrackerComponent|GetFortGameplayDataTrackerComponentManager))
      (Utilities|IsValid _returnvalue
        (:Is%20Valid
          (Class|FortGameplayDataTrackerComponentManager|UnRegisterTrackedStateActor _returnvalue self))))))

(event EventBeginPlay
  (|CallParentFunction)
  (Networking|CanExecuteCosmeticEvents
    (:True
      (bind _returnvalue (FortGameplayDataTrackerComponent|GetFortGameplayDataTrackerComponentManager))
      (Utilities|IsValid _returnvalue
        (:Is%20Valid
          (Class|FortGameplayDataTrackerComponentManager|RegisterTrackedStateActor _returnvalue self (CallFunction|GetStateTagfromAlertLevel)))
        (:Is%20Not%20Valid
          _)))
    (:False
      _))
  ;[1 0 0] HACK ANIM TEST FOR FULL BODY REACTS
  (Ability|EquipFortAbilitySet self (Variables|Default|GetHitReactAbilitySetHACK) self) ; Give Full Body Hit React Ability
  (CallFunction|InitializeAnimInstanceSettings) ; Needed to prevent full anim update from running between BeginPlay and PostLoad
  ;[1 1 1] Clear up "!" and "?" VFX immediately on DBNO
  (EventDispatchers|AssignOnDBNOStateChanged self (delegate Custom|OnDBNOStateChanged_Event) :!class /Script/FortniteGame.FortPawn))

(event Custom|OnDBNOStateChanged_Event
  (params (FortPawn :type Fort%20Pawn%20Object%20Reference) (bInIsDBNO :type Boolean))
  ;[1 1 1] Clear up "!" and "?" VFX immediately on DBNO
  ;;[1 1 1] Enable full movement simulation when going into DBNO to fix pops with throw/drop/interrogation animations
  (CallFunction|SetShouldSkipMovementFullSimulation (not bInIsDBNO))
  (if bInIsDBNO
    (CallFunction|SetAlertedStateGameplayCue false)
    (CallFunction|PlayVoiceComponentSoundbyTag "(TagName=\"Pawn.Athena.NPC.Sound.Voice.DBNO\")")
    (GameplayEffects|ApplyGameplayEffectToSelf :self (Ability|GetAbilitySystemComponent self) :GameplayEffectClass (Variables|Default|GetGE_NPC_Status_Stressed_Infinite) :Level 0.0 :!pinned true) ; DBNO = Infinite Stress\nPropagate to other Henchmen.
    (elif (and (Utilities|Select :Index (Variables|Default|GetAlertLevel) :Unaware true :Suspicious false :LostTarget false :Threatened false :Count false) (not (FortPawn|IsDead self)))
      (Networking|SwitchHasAuthority
        (:Authority
          (GameplayEffects|ApplyGameplayEffectToSelf :self (Ability|GetAbilitySystemComponent self) :GameplayEffectClass (Variables|Default|GetGE_NPC_Status_Stressed_Cooldown) :Level 0.0 :!pinned true) ; Recover from DBNO = Cooldown stress that bleeds off.\nWill run around in patrol for a short time if he has no target.
        )))))

(event EventOnDamageServer
  (params (Damage :type Float%20%28single-precision%29) (DamageTags :type Gameplay%20Tag%20Container%20Structure%20%28by%20ref%29) (Momentum :type Vector :default "0, 0, 0") (HitInfo :type Hit%20Result%20Structure%20%28by%20ref%29) (InstigatedBy :type Controller%20Object%20Reference) (DamageCauser :type Actor%20Object%20Reference) (EffectContext :type Gameplay%20Effect%20Context%20Handle%20Structure))
  ;[0.5915 0 1] Despawn due to Storm Damage
  (|CallParentFunction Damage DamageTags Momentum HitInfo InstigatedBy DamageCauser EffectContext)
  (if (GameplayTags|HasTag DamageTags "(TagName=\"Gameplay.Damage.OutsideSafeZone\")")
    (Ability|SendGameplayEventtoActor self "(TagName=\"NPC.Ability.NonAttack.TeleportOut\")") ; Teleport out and despawn when taking storm damage.
  ))

(event GameplayCue|GameplayCue.Player.Interrogation.Voice.PickedUp
  (params (EventType :type EGameplayCueEvent%20Enum :default "OnActive") (Parameters :type Gameplay%20Cue%20Parameters%20Structure%20%28by%20ref%29))
  ;[1 1 1] Interrogation GCs from Interrogation system
  (Networking|CanExecuteCosmeticEvents
    (:True
      (CallFunction|PlayVoiceComponentSoundbyTag "(TagName=\"Pawn.Athena.NPC.Sound.Voice.Interrogate\")"))))

(event EventOnCharacterCustomizationCompleted
  (params (Pawn :type Fort%20Player%20Pawn%20Object%20Reference))
  ;[1 1 1] Anim Setup
  (Components|SkeletalMesh|AnimationBlueprintLinking|LinkAnimClassLayers (Variables|Character|GetMesh) (Variables|Animation|GetPatrolLayerAnimBP))
  (Default|UpdatePatrolAnimSet (Variables|Weapon|GetCurrentWeapon Pawn))
  (CallFunction|InitializeAnimInstanceSettings))

(event Custom|IdlePatrolVO
  ;[1 1 1] IDLE VO
  (Networking|CanExecuteCosmeticEvents
    (:True
      (CallFunction|PlayVoiceComponentSoundbyTag "(TagName=\"Pawn.Athena.NPC.Sound.Voice.Idle\")"))))

(event EventOnWeaponEquipped
  (params (NewWeapon :type Fort%20Weapon%20Object%20Reference) (PrevWeapon :type Fort%20Weapon%20Object%20Reference))
  ;[1 1 1] Anim Setup
  (|CallParentFunction NewWeapon PrevWeapon)
  (Default|UpdatePatrolAnimSet NewWeapon))

(event OnComponentHit(CapsuleComponent)
  (params (HitComponent :type Primitive%20Component%20Object%20Reference) (OtherActor :type Actor%20Object%20Reference) (OtherComp :type Primitive%20Component%20Object%20Reference) (NormalImpulse :type Vector :default "0, 0, 0") (Hit :type Hit%20Result%20Structure%20%28by%20ref%29))
  ;[1 1 1] If a disguised player bumps into a Henchman, remove the disguise.
  (if (Utilities|NotEqual(Object) OtherActor self)
    (bind _asplayer_pawn_athena (Utilities|Casting|CastToPlayerPawn_Athena OtherActor)) ; Only care about Pawns.
    (if (GameplayTags|HasMatchingGameplayTag _asplayer_pawn_athena (Variables|Default|GetTC_DisguiseTag))
      (GameplayEffects|ApplyGameplayEffectToTarget :self (Ability|GetAbilitySystemComponent _asplayer_pawn_athena) :GameplayEffectClass (Variables|Default|GetGE_RemoveDisguise) :Target (Ability|GetAbilitySystemComponent _asplayer_pawn_athena) :Level 0.0) ; Remove the disguise.
    )))

(event EventOnDeathPlayEffects
  (params (Damage :type Float%20%28single-precision%29) (DamageTags :type Gameplay%20Tag%20Container%20Structure%20%28by%20ref%29) (Momentum :type Vector :default "0, 0, 0") (HitInfo :type Hit%20Result%20Structure%20%28by%20ref%29) (InstigatedBy :type Fort%20Pawn%20Object%20Reference) (DamageCauser :type Actor%20Object%20Reference) (EffectContext :type Gameplay%20Effect%20Context%20Handle%20Structure))
  ;[1 1 1] Make sure we remove "!" "?" when dead.
  ;
  ; Mostly used for LTM where we don't DBNO these guys.
  (|CallParentFunction Damage DamageTags Momentum HitInfo InstigatedBy DamageCauser EffectContext)
  (CallFunction|SetAlertedStateGameplayCue false))

(event EventOnDeathServer
  (params (Damage :type Float%20%28single-precision%29) (DamageTags :type Gameplay%20Tag%20Container%20Structure%20%28by%20ref%29) (Momentum :type Vector :default "0, 0, 0") (HitInfo :type Hit%20Result%20Structure%20%28by%20ref%29) (InstigatedBy :type Controller%20Object%20Reference) (DamageCauser :type Actor%20Object%20Reference) (EffectContext :type Gameplay%20Effect%20Context%20Handle%20Structure))
  ;[1 0.5677 0.5677] Small Hack to drop Ammo on Death since if we give them in the inventory, they will use the ammo.
  (|CallParentFunction Damage DamageTags Momentum HitInfo InstigatedBy DamageCauser EffectContext)
  (if (Variables|Default|GetShouldDropAmmo)
    (Inventory|SpawnPickupInWorld (Variables|Default|GetAmmoDefToSpawn) (Variables|Default|GetAmmoCountToDrop) (Transformation|GetActorLocation self) "0.000000,0.000000,1.000000" 0 true true false -1 "Tossed" "PlayerElimination" 0 false)))

(fn AlertLevelChanged
  (params (OldAlertLevel :type EAlertLevel%20Enum :default "Unaware") (NewAlertLevel :type EAlertLevel%20Enum :default "Unaware"))
  (Variables|Default|SetOldAlertLevel OldAlertLevel)
  (Variables|Default|SetAlertLevel NewAlertLevel))

(fn DebugControlResistanceInfo
  (bind (_bsuccessfullyfoundattribute _returnvalue) (Ability|Attribute|GetFloatAttribute self "(AttributeName=\"ControlResistanceRegenRate\",Attribute=,AttributeOwner=None)"))
  (bind (_bsuccessfullyfoundattribute_2 _returnvalue_2) (Ability|Attribute|GetFloatAttribute self "(AttributeName=\"ControlResistance\",Attribute=,AttributeOwner=None)"))
  (bind (_bsuccessfullyfoundattribute_3 _returnvalue_3) (Ability|Attribute|GetFloatAttribute self "(AttributeName=\"MaxControlResistance\",Attribute=,AttributeOwner=None)"))
  (bind (_bsuccessfullyfoundattribute_4 _returnvalue_4) (Ability|Attribute|GetFloatAttribute self "(AttributeName=\"ControlResistanceRegenDelay\",Attribute=,AttributeOwner=None)"))
  (if (not (Variables|Health|GetbIsDBNO))
    (switch Utilities|FlowControl|Switch|SwitchonEFortControlRecoveryBehavior (Variables|ControlResistance|GetControlRecoveryBehavior)
      (:DefaultControl
        _)
      (:LimitedControl
        _)
      (:ChainControl
        (bind _data (Utilities|FlowControl|MergeExecutionwithData (stale _data_2) "Chain" :!label _label)
          (:Then
            (Networking|SwitchHasAuthority
              (:Authority
                (Development|PrintString (Utilities|String|BuildString(Float) (Utilities|String|BuildString(Float) (Utilities|String|BuildString(Float) (Utilities|String|BuildString(Float) (Utilities|String|Append (Utilities|String|BuildString(Object) "===== " "PAWN = " self " | ") _data " | ") "CR = " _returnvalue_2 "/") "" _returnvalue_3 " | ") "RegRate = " _returnvalue " | ") "RegDelay = " _returnvalue_4 " ====="))))))))
    (bind _data_2 (Utilities|FlowControl|MergeExecutionwithData "Default" "Limited")
      (:Then
        (goto _label)))))

(fn GetAdditiveHitReactsMontage
  (bind _currentpatrolanimasset (Variables|Animation|GetCurrentPatrolAnimAsset))
  (bind _animset (Variables|/Script/FortniteGame.FortPatrolAnimAsset|GetAnimSet _currentpatrolanimasset))
  (bind (_patrolidlepose _patrolidleadditive _additivehitreactmontage) (Utilities|Struct|BreakFortAnimInputPatrolAnimSet _animset))
  (Utilities|IsValid _currentpatrolanimasset
    (:Is%20Valid
      (Utilities|IsValid _additivehitreactmontage
        (:Is%20Valid
          (return _additivehitreactmontage))
        (:Is%20Not%20Valid
          _)))
    (:Is%20Not%20Valid
      _))
  (return (Variables|Animation|GetDefaultAdditiveHitReactionMontage)))

(fn GetCurrentFullbodyHitReactionMontage
  (bind _currentpatrolanimasset (Variables|Animation|GetCurrentPatrolAnimAsset))
  (bind _animset (Variables|/Script/FortniteGame.FortPatrolAnimAsset|GetAnimSet _currentpatrolanimasset))
  (bind (_patrolidlepose _patrolidleadditive _additivehitreactmontage _fullbodyhitreactmontage) (Utilities|Struct|BreakFortAnimInputPatrolAnimSet _animset))
  (if (Math|Boolean|ANDBoolean (not (Variables|Character|GetbIsCrouched)) (Pawn|Components|CharacterMovement|IsWalking (Variables|Character|GetCharacterMovement)) (Utilities|IsValid _currentpatrolanimasset :!class /Script/Engine.KismetSystemLibrary))
    (Utilities|IsValid _fullbodyhitreactmontage
      (:Is%20Valid
        (return _fullbodyhitreactmontage))
      (:Is%20Not%20Valid
        _)))
  (return (Variables|Animation|GetDefaultFullbodyHitReactionMontage)))

(fn GetStateTagFromAlertLevel
  (return (Utilities|Select :Index (Variables|Default|GetAlertLevel) :Unaware "(TagName=\"NPC.AlertLevel.Unaware\")" :Suspicious "(TagName=\"NPC.AlertLevel.Alerted\")" :LostTarget "(TagName=\"NPC.AlertLevel.Aggressive\")" :Threatened "(TagName=\"NPC.AlertLevel.Aggressive\")")))

(fn InitializeAnimInstanceSettings
  (bind _asfort_player_anim_instance (Utilities|Casting|CastToFortPlayerAnimInstance (Components|SkeletalMesh|GetAnimInstance (Variables|Character|GetMesh))))
  (Variables|/Script/FortniteGame.FortPlayerAnimInstance|SetbShouldProcessFullAnimUpdate (Variables|Animation|GetbShouldDoFullAnimationUpdate) _asfort_player_anim_instance)
  (Variables|/Script/FortniteGame.FortAnimInstance|SetbShouldPredictLanding false _asfort_player_anim_instance))

(fn OnRep_AlertLevel
  (bind _self self)
  (bind _returnvalue (AI|GetBlackboard _self))
  (bind _returnvalue_2 (Utilities|Name|MakeLiteralName "AIEvaluator_Global_TargetActor"))
  (bind _returnvalue_3 (AI|Components|Blackboard|GetValueasObject _returnvalue _returnvalue_2))
  (bind _alertlevel (Variables|Default|GetAlertLevel))
  (bind _oldalertlevel (Variables|Default|GetOldAlertLevel))
  (bind _oldalertlevel_2 (Variables|Default|GetOldAlertLevel))
  (bind _oldalertlevel_3 (Variables|Default|GetOldAlertLevel))
  (Networking|CanExecuteCosmeticEvents
    (:True
      (bind _returnvalue_4 (FortGameplayDataTrackerComponent|GetFortGameplayDataTrackerComponentManager))
      (Utilities|IsValid _returnvalue_4
        (:Is%20Valid
          (Class|FortGameplayDataTrackerComponentManager|SetTrackedActorState _returnvalue_4 self (CallFunction|GetStateTagfromAlertLevel)))
        (:Is%20Not%20Valid
          _)))
    (:False
      _))
  (if (Math|Boolean|ORBoolean (Variables|Health|GetbIsDBNO) (GameplayTags|HasMatchingGameplayTag self "(TagName=\"Gameplay.Status.DBNO\")") (FortPawn|IsDead self))
    (CallFunction|SetAlertedStateGameplayCue false)
    (GameplayEffects|ApplyGameplayEffectToSelf :self (Ability|GetAbilitySystemComponent self) :GameplayEffectClass (Variables|Default|GetGE_NPC_Status_Stressed_Infinite) :Level 0.0 :!pinned true) ; INFINITE STRESS\n\nLet Patrol system know we were recently alerted, so we patrol faster for a while. "Stressed"
    (elif (Utilities|Enum|Equal(Enum) _alertlevel "Unaware")
      (CallFunction|SetAlertedStateGameplayCue false)
      (if (Utilities|Select :Index _oldalertlevel_2 :Unaware false :Suspicious false :LostTarget true :Threatened true :Count false)
        (CallFunction|PlayVoiceComponentSoundbyTag "(TagName=\"Pawn.Athena.NPC.Sound.Voice.ReturnToIdle\")" :!pinned true) ; Unaware from Combat (de-escalation)
        (elif (Utilities|Select :Index _oldalertlevel_2 :Unaware false :Suspicious true :LostTarget false :Threatened false :Count false)
          (CallFunction|PlayVoiceComponentSoundbyTag "(TagName=\"Pawn.Athena.NPC.Sound.Voice.ReturnToIdle\")" :!pinned true) ; Unaware from Alerted (de-escalation)
        ))
      (GameplayEffects|ApplyGameplayEffectToSelf :self (Ability|GetAbilitySystemComponent self) :GameplayEffectClass (Variables|Default|GetGE_NPC_Status_Stressed_Cooldown) :Level 0.0 :!pinned true) ; STRESS THAT COOLS DOWN\n\nLet Patrol system know we were recently alerted, so we patrol faster for a while. "Stressed"
      (elif (Utilities|Enum|Equal(Enum) _alertlevel "Suspicious")
        (CallFunction|SetAlertedStateGameplayCue true)
        (if (Utilities|Select :Index _oldalertlevel_3 :Unaware false :Suspicious false :LostTarget true :Threatened true :Count false)
          (CallFunction|PlayVoiceComponentSoundbyTag "(TagName=\"Pawn.Athena.NPC.Sound.Voice.LostTarget\")" :!pinned true) ; Alerted from Combat (de-escalation)
          (elif (Utilities|Select :Index _oldalertlevel_3 :Unaware true :Suspicious false :LostTarget false :Threatened false :Count false)
            (CallFunction|PlayVoiceComponentSoundbyTag "(TagName=\"Pawn.Athena.NPC.Sound.Voice.Suspicious\")" :!pinned true) ; Alerted from Unaware (ESCALATION)
            (Ability|SendGameplayEventtoActor self "(TagName=\"Event.Animation.Alerted\")" :!pinned true) ; HACK TEST FOR HIGHLIGHTING ESCALATING ALERT LEVELS WITH TRIGGERED ANIMATIONS
          ))
        (elif (or (Utilities|Enum|Equal(Enum) _alertlevel "LostTarget") (Utilities|Enum|Equal(Enum) _alertlevel "Threatened"))
          (CallFunction|SetAlertedStateGameplayCue true true)
          (if (Utilities|Select :Index _oldalertlevel :Unaware false :Suspicious true :LostTarget false :Threatened false :Count false)
            (CallFunction|PlayVoiceComponentSoundbyTag "(TagName=\"Pawn.Athena.NPC.Sound.Voice.FullAlert\")" :!pinned true) ; Combat from Alerted (ESCALATION)
            (elif (Utilities|Select :Index _oldalertlevel :Unaware true :Suspicious false :LostTarget false :Threatened false :Count false)
              (CallFunction|PlayVoiceComponentSoundbyTag "(TagName=\"Pawn.Athena.NPC.Sound.Voice.FullAlert\")" :!pinned true) ; Combat from Unaware (ESCALATION)
            ))
          (Networking|SwitchHasAuthority
            (:Authority
              (Utilities|IsValid _returnvalue
                (:Is%20Valid
                  (Utilities|IsValid _returnvalue_3
                    (:Is%20Valid
                      (bind _asactor (Utilities|Casting|CastToActor _returnvalue_3)
                        (:then
                          (if (> (Class|AbilityFunctions|GetDistanceBetweenActors _asactor self) (Variables|Animation|GetMinDistanceToTargetToPlayCombatAlertAnimation))
                            (Ability|SendGameplayEventtoActor self "(TagName=\"Event.Animation.Combat\")" :!pinned true) ; HACK TEST FOR HIGHLIGHTING ESCALATING ALERT LEVELS WITH TRIGGERED ANIMATIONS
                          ))
                        (:CastFailed
                          _)))
                    (:Is%20Not%20Valid
                      _)))
                (:Is%20Not%20Valid
                  _))))))))
  (GameplayEffects|ApplyGameplayEffectToSelf :self (Ability|GetAbilitySystemComponent self) :GameplayEffectClass (Variables|Default|GetGE_NPC_Status_Stressed_Infinite) :Level 0.0 :!pinned true) ; INFINITE STRESS\n\nLet Patrol system know we were recently alerted, so we patrol faster for a while. "Stressed"
)

(fn PlayVoiceComponentSoundByTag
  (params (SoundBankTag :type Gameplay%20Tag%20Structure))
  (bind _alertsfxmap (Variables|Default|GetAlertSFXMap))
  (bind (_value _returnvalue) (Utilities|Map|Find _alertsfxmap SoundBankTag))
  (Utilities|FlowControl|Sequence
    (:then_0
      (if (Math|Boolean|ORBoolean (not (Variables|Health|GetbIsDBNO)) (GameplayTags|Equal(GameplayTag) SoundBankTag "(TagName=\"Pawn.Athena.NPC.Sound.Voice.Interrogate\")") (GameplayTags|Equal(GameplayTag) SoundBankTag "(TagName=\"Pawn.Athena.NPC.Sound.Voice.DBNO\")"))
        (Networking|CanExecuteCosmeticEvents
          (:True
            (Class|NPCVoiceComponent|PlayVoiceLine (Variables|Default|GetBP_VoiceComponent_MANG) SoundBankTag (Variables|Default|GetVOAudioComponent))))))
    (:then_1
      (if _returnvalue
        (Audio|PlaySoundatLocation _value (Transformation|GetActorLocation))))))

(fn SetAlertedStateGameplayCue
  (params (Alerted :type Boolean) (Combat :type Boolean))
  (bind _returnvalue (Math|Integer|SelectInt 1 0 Combat))
  (bind _returnvalue_2 (Ability|GameplayCue|MakeGameplayCueParameters :NormalizedMagnitude 0.0 :RawMagnitude 0.0 :Location "0,0,0" :Normal "0,0,0" :GameplayEffectLevel _returnvalue :AbilityLevel 1 :bReplicateLocationWhenUsingMinimalRepProxy false))
  (bind _self self)
  (bind _gameplaycuealertstate (Variables|Default|GetGameplayCueAlertState))
  (if Alerted
    (Projectile|AddGameplayCue%28Looping%29Local _self _gameplaycuealertstate _returnvalue_2 :!pinned true) ; "!" or "?" On
    (else
      (Projectile|RemoveGameplayCue%28Looping%29Local _self _gameplaycuealertstate _returnvalue_2 :!pinned true) ; "!" or "?" Off
    )))

(fn SetAmmoToSpawnOnDeath
  (params (AmmoDefinition :type Fort%20World%20Item%20Definition%20Object%20Reference) (Count :type Integer) (ShouldDropAmmo :type Boolean))
  (Variables|Default|SetAmmoDefToSpawn AmmoDefinition)
  (Variables|Default|SetAmmoCountToDrop Count)
  (Variables|Default|SetShouldDropAmmo ShouldDropAmmo)
  (return))

(fn UpdatePatrolAnimSet
  (params (NewWeapon :type Fort%20Weapon%20Object%20Reference))
  (bind _mesh (Variables|Character|GetMesh))
  (bind _returnvalue (Components|SkeletalMesh|GetAnimInstance _mesh))
  (Utilities|IsValid _returnvalue
    (:Is%20Valid
      (bind _asfort_patrol_anim_layer (Utilities|Casting|CastToFortPatrolAnimLayer (Animation|LinkedAnimGraphs|GetLinkedAnimLayerInstancebyClass _returnvalue (Variables|Animation|GetPatrolLayerAnimBP) false)))
      (bind _output_get (LocalVariables|SetFoundAnimAsset (Class|FortPatrolAnimLayer|GetPatrolAnimSetfromWeapon _asfort_patrol_anim_layer (Variables|Animation|GetPatrolAnimSetPairs) NewWeapon)))
      (Class|FortPatrolAnimLayer|SetPatrolAnimAsset _asfort_patrol_anim_layer _output_get)
      (Variables|Animation|SetCurrentPatrolAnimAsset (LocalVariables|GetFoundAnimAsset)))))

(fn ConstructionScript
  (|CallParentFunction))
